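# StencilMaskCompare is a library material function that builds a mask for pixels whose custom stencil value matches a given ID.
# It reads the red channel of the CustomStencil scene texture and compares it to the scalar ID input through an If node.
# The output is white on a match and black when the stencil value is higher or lower.
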MaterialFunction /Game/Packages/Fortress_SharedMaterials/Stencil_ID_Definitions/StencilMaskCompare
graph {
 properties {bExposeToLibrary=true, description=Compares a float to the stencil buffer. White if true - black if false.}
 nodes [
  n1 = SceneTexture() {sceneTextureId=PPI_CustomStencil}
  n2 = Constant()
  FunctionOutput(None=If(
    A=ComponentMask(None=@n1.Color) {r=true}
    B=FunctionInput() {bUsePreviewValueAsDefault=true, inputName=ID, inputType=FunctionInput_Scalar}
    "A > B"=@n2
    "A == B"=Constant() {r=1}
    "A < B"=@n2
   ))
 ]
}
